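# Sound cue for the Outlander charged fist ability's activation.
# A Random node picks one of two equally weighted fire wave variations on each play.
# It routes through the melee weapon sound class with the shared ability-radius attenuation and a slightly reduced volume.
# The cue is primed on load.
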SoundCue /Game/Sounds/Fort_Gadgets/Outlander/Ability_ChargedFist_Activate_Cue
root=Random_2
nodes [
 {id=Random_2, children [WavePlayer_5,WavePlayer_6], properties {weights [1,1]}}
 {id=WavePlayer_5, properties {soundWaveAssetPtr=&/Ability_ChargedFist_Fire_01}}
 {id=WavePlayer_6, properties {soundWaveAssetPtr=&/Ability_ChargedFist_Fire_02}}
]
properties {
 attenuationSettings=/Game/Sounds/Attenuation/Fort_Ability_Radius
 bPrimeOnLoad=true
 maxDistance=1200
 soundClassObject=&Fort_WeaponMelee_PSM
 volumeMultiplier=0.9
}
