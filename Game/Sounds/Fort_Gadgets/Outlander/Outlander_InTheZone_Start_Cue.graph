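# Sound cue for the start of the Outlander "In The Zone" gadget, in the Save the World gadget sound class.
# A single wave player plays one start sound at reduced volume, with no randomization or mixing.
# It uses the shared ability-radius attenuation and primes on load.
# A concurrency override prevents new instances once the limit is reached instead of cutting off playing ones.
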SoundCue /Game/Sounds/Fort_Gadgets/Outlander/Outlander_InTheZone_Start_Cue
root=WavePlayer_2
nodes [{id=WavePlayer_2, properties {soundWaveAssetPtr=&/inthezone_start_01}}]
properties {
 attenuationSettings=/Game/Sounds/Attenuation/Fort_Ability_Radius
 bOverrideConcurrency=true
 bPrimeOnLoad=true
 maxDistance=1200
 soundClassObject=/Game/Sounds/FortSoundClassesAndModes/Fort_Gadgets_StW
 volumeMultiplier=0.4
 concurrencyOverrides {resolutionRule=PreventNew}
}
